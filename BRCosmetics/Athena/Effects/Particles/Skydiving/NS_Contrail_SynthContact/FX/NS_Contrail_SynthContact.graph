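# Niagara skydiving contrail cosmetic SynthContact, trailing red light streaks, rays and speed lines from the character.
# FN_Contrail_Setup runs at system level, and most emitters sample the skeletal mesh, using partition indices to pick bones or sockets.
# Burst-spawned persistent meshes, with lights and glow sprites on hand and feet streaks, follow those points, while rate-spawned sprites add moving lines.
# ParaGlideLeanAlpha, dive state, time of day and front-end overrides adjust orientation, motion and intensity.
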NiagaraSystem /BRCosmetics/Athena/Effects/Particles/Skydiving/NS_Contrail_SynthContact/FX/NS_Contrail_SynthContact
dependencies {
 usedDataInterfaces [
  &NiagaraDataInterfaceArrayFloat3
  &NiagaraDataInterfaceCurve
  /Script/Niagara.NiagaraDataInterfaceArrayFloat4
  &NiagaraDataInterfaceVector2DCurve
  &NiagaraDataInterfaceColorCurve
  &NiagaraDataInterfaceSkeletalMesh
 ]
 usedDynamicInputs []
 usedModules [
  &self:SystemSpawnScript.SystemScriptSource.SystemScriptGraph.NiagaraNodeAssignment_0.SetVariables_#1
  &SystemState
  /Game/Effects/Niagara/Modules/Contrails/FN_Contrail_Setup
 ]
}
emitters {
 LightRay {
  emitterUpdateScript [
   &EmitterState(
    Inactive Response=::NewEnumerator1
    Life Cycle Mode=::NewEnumerator1
    Loop Behavior=::NewEnumerator1
    Loop Duration Mode=::NewEnumerator1
    UseLoopDelay=false
   )
   &SpawnRate(SpawnRate=4, Use Spawn Probability=false) {enabled=false}
   /SetVariables(Emitter.DistanceFractionThreshold:NiagaraBool=&SetBoolByFloatComparison(A=$Emitter.Scalability.DistanceFraction, B=0.9, Comparison Type=::NewEnumerator6)) {isSetParameters=true}
   &SpawnBurst_Instantaneous(Spawn Count=4, Use Spawn Probability=false, UseLoopCountLimit=false)
   &Timeline(Play=$Emitter.DistanceFractionThreshold, Play Rate=4, Rewind Rate=4) {module=Timeline001}
  ]
  particleSpawnScript [
   &PartitionParticles(Number Of Partitions=4)
   &InitializeParticle(
    Color Mode=::NewEnumerator0
    Lifetime=&RandomRangeFloat(Minimum=1.7, Maximum=2)
    Mesh Scale Mode=::NewEnumerator1
    Mesh Uniform Scale Max=0.8
    Mesh Uniform Scale Min=0.6
    UsePositionOffset=false
    Write Material Random=true
   )
   &InitialMeshOrientation(
    Mesh Axis To Orient=&FN_FrontEndVector3Modifier(Gameplay=<1 0 0>, Front End=<1 0 0>)
    Mesh Orientation Mode=::NewEnumerator2
    Orientation Coordinate Space=::Local
    Rotation=&FN_FrontEndVector3Modifier(
     Gameplay=&SelectVectorFromArray(
      Direct Array Index=$Particles.PartitionParticles.Partition
      Vector Selection Array=DataInterface() {propertyValues {FloatData [<0.03 0.01 0.09>, <0.03 -0.01 -0.09>, <0 0 0>, <0 0 0>]}}
     )
     Front End=&SelectVectorFromArray(
      Direct Array Index=$Particles.PartitionParticles.Partition
      Vector Selection Array=DataInterface() {propertyValues {FloatData [<0.05 0.03 0.09>, <0.05 -0.03 -0.09>, <0 0 0>, <0 0 0>]}}
     )
    )
    Rotation Coordinate Space=::NewEnumerator2
    Use Rotation Vector=true
   ) {enabled=false, module=InitialMeshOrientation001}
  ]
  particleUpdateScript [
   &ParticleState(Kill Particles When Lifetime Has Elapsed=false, Let Infinitely Lived Particles Die When Emitter Deactivates=true)
   &UpdateMeshOrientation(
    Facing Direction=&Lerp_Vector(
     A=&SelectVectorFromArray(
      Direct Array Index=$Particles.PartitionParticles.Partition
      Vector Selection Array=DataInterface() {propertyValues {FloatData [<1 0 0>, <1 0 0>]}}
     )
     B=&SelectVectorFromArray(
      Direct Array Index=$Particles.PartitionParticles.Partition
      Vector Selection Array=DataInterface() {propertyValues {FloatData [<1 0 0>, <1 0 0>]}}
     )
     Alpha=$User.ParaGlideLeanAlpha
    )
    Facing Interpolation=false
    Orientation Method=::NewEnumerator1
    Override Quat=false
    TransformFacing=false
    TransformUp=false
    Up Direction:Vector3f=&FN_FrontEndVector3Modifier(Gameplay=<0 0.4 8>, Front End=&Lerp_Vector(A=<0 0.3 1>, B=<0 -0.4 1>, Alpha=$User.ParaGlideLeanAlpha))
   ) {enabled=false, module=UpdateMeshOrientation002}
   &UpdateMeshOrientation(
    Facing Interpolation=false
    Orientation Method=::NewEnumerator1
    Override Quat=false
    TransformFacing=false
    TransformUp=false
    Up Direction:Vector3f=&FN_FrontEndVector3Modifier(Gameplay=<0 0 1>, Front End=&Lerp_Vector(A=<0 -0.3 1>, B=<0 -0.7 1>, Alpha=$User.ParaGlideLeanAlpha))
   )
   &Timeline(
    Play Rate=&Lerp_Float(A=1, B=2, Alpha=&ClampFloat(Float=$User.ParaGlideLeanAlpha))
    Timeline End=10000000
   )
   &ScaleMeshSize(Scale Factor=&MakeVector(
     X=1
     Y=1
     Z=&FloatFromCurve(
      FloatCurve=DataInterface() {propertyValues {Curve {keys [
          {arriveTangent=5.7371, interpMode=RCIM_Cubic, leaveTangent=5.7371, tangentMode=RCTM_User, time=0, value=0}
          {arriveTangent=-0.9106, interpMode=RCIM_Cubic, leaveTangent=-0.9106, tangentMode=RCTM_User, time=0.76, value=1.025}
          <at 1 1>
         ]}}}
      CurveIndex=$Emitter.Age
      Scale Curve=$Emitter.Timeline001.NormalizedTime
     )
    ))
   &DynamicMaterialParameters(
    Index 0 Param 2=$System.Dive
    Index 0 Param 3=&RandomRangeFloat(Minimum=0, Maximum=1)
    Index 0 Param 4=&RandomRangeFloat(Minimum=0, Maximum=1)
    Module.Index0Param0Write=true
    Module.Index0Param1Write=true
    Module.Index0Param2Write=true
    Module.Index0Param3Write=true
   ) {module=DynamicMaterialParameters001}
   &SkeletalMeshLocation(
    Bone / Socket Index=$Particles.PartitionParticles.Partition
    Bone Sampling Mode=::NewEnumerator4
    MaskSpawnGroup=false
    Orientation Sampling=::NewEnumerator1
    Skeletal Mesh=$System.Skeletal Mesh
   )
   &ScaleMeshSize(Scale Factor=&FN_FrontEndVector3Modifier(Gameplay=<0.7 0.7 0.8>, Front End=<0.65 0.65 0.75>)) {module=ScaleMeshSize001}
   &ScaleMeshSize(Scale Factor=&SelectVectorFromArray(
     Direct Array Index=$Particles.PartitionParticles.Partition
     Vector Selection Array=DataInterface() {propertyValues {FloatData [<1 1 1>, <1 1 1>, <0.5 0.5 0.4>, <0.5 0.5 0.4>]}}
    )) {module=ScaleMeshSize002}
   &ScaleColor(
    Scale Alpha=&FloatFromCurve(
     FloatCurve=DataInterface() {propertyValues {Curve {keys [<at 0 1 RCIM_Cubic>,<at 0.2 1 RCIM_Cubic>,<at 1 1>]}}}
     Scale Curve=&Lerp_Float(A=0.4, B=0.55, Alpha=$System.Dive)
    )
    Scale RGB=&MakeVectorFromLinearColorRGB(Color=&FNTimeOfDayColor(
      Night Color=<rgba 0.6 0.6 0.6 1>
      Morning Color=<rgba 0.65 0.65 0.65 1>
      Day Color=<rgba 0.9 0.9 0.9 1>
      Evening Color=<rgba 0.65 0.65 0.65 1>
     ))
    ScaleA=true
    ScaleRGB=true
   )
   &ScaleColor(Scale Alpha=&FN_FrontEndFloatModifier(Gameplay=1, Front End=0.6), ScaleA=true, ScaleRGB=true) {module=ScaleColor001}
  ]
  rendererClasses [&NiagaraMeshRendererProperties]
  renderers [{rendererClass=&NiagaraMeshRendererProperties, rendererIndex=0}]
 }
 LightStreak_Feet {
  emitterUpdateScript [
   &EmitterState(Life Cycle Mode=::NewEnumerator1, Loop Behavior=::NewEnumerator1, Loop Duration Mode=::NewEnumerator1, UseLoopDelay=false)
   &SpawnBurst_Instantaneous(Spawn Count=2, Use Spawn Probability=false, UseLoopCountLimit=false)
  ]
  particleSpawnScript [
   &PartitionParticles()
   &InitializeParticle(
    Color=<rgba 1 0.5473 0.4674 0.25>
    Mesh Scale=<1 1 0.4>
    Mesh Scale Mode=::NewEnumerator4
    Sprite Size Mode=::NewEnumerator3
    Uniform Sprite Size=30
    UsePositionOffset=false
    Write Material Random=true
   )
   &InitialMeshOrientation(
    Mesh Orientation Mode=::NewEnumerator7
    Rotation=&SelectVectorFromArray(
     Direct Array Index=$Particles.PartitionParticles.Partition
     Vector Selection Array=DataInterface() {propertyValues {FloatData [<0 0 0>, <0 0 0>]}}
    )
    Use Rotation Vector=true
   ) {enabled=false}
   &DynamicMaterialParameters(
    Index 0 Param 1=80
    Index 0 Param 2=5
    Index 0 Param 3=&RandomRangeFloat(Minimum=-1, Maximum=1)
    Index 0 Param 4=1
    Module.Index0Param0Write=true
    Module.Index0Param1Write=true
    Module.Index0Param2Write=true
    Module.Index0Param3Write=true
   )
  ]
  particleUpdateScript [
   &ParticleState(Kill Particles When Lifetime Has Elapsed=false, Let Infinitely Lived Particles Die When Emitter Deactivates=true)
   &ScaleMeshSize(Scale Factor=&Lerp_Vector(
     A=<1 1 0>
     B=<1 1 1>
     Alpha=&FloatFromCurve(
      FloatCurve=DataInterface() {propertyValues {Curve {keys [{interpMode=RCIM_Cubic, leaveTangent=0.4055, tangentMode=RCTM_Break, time=0, value=0}, <at 0.5 0.95>]}}}
      CurveIndex=$Emitter.Age
     )
    ))
   &SkeletalMeshLocation(
    Bone / Socket Index=&Add_Integer(A=$Particles.PartitionParticles.Partition, B=2)
    Bone Sampling Mode=::NewEnumerator4
    MaskSpawnGroup=false
    Orientation Sampling=::NewEnumerator1
    Skeletal Mesh=$System.Skeletal Mesh
   )
   &UpdateMeshOrientation(
    Facing Interpolation=false
    Orientation Method=::NewEnumerator1
    Override Quat=false
    TransformFacing=false
    TransformUp=false
    Up Direction:Vector3f=&FN_FrontEndVector3Modifier(Gameplay=<0 0 1>, Front End=&Lerp_Vector(A=<0 -0.3 1>, B=<0 -0.6 1>, Alpha=$User.ParaGlideLeanAlpha))
   )
   /SetVariables(
    Particles.BorderColor:LinearColor=<rgba 2 0.0367 0 0.5>
    Particles.BorderDyn0:Vector4f=<1 0.8002 0 0>
    Particles.BorderDyn1:Vector4f=/Niagara/DynamicInputs/Arrays/SelectVector4_FromArray(
     Array Sampling Mode:ENiagaraArraySamplingMode=::NewEnumerator1
     Direct Array Index:NiagaraInt32=$Particles.PartitionParticles.Partition
     Interpolation:NiagaraFloat=?
     Direct Array Mode:ENiagara_WrapClamp=::NewEnumerator0
     Evaluation Type:ENiagaraRandomnessEvaluation=::NewEnumerator0
     Normalized Interpolation:NiagaraBool=true
     Recalculate Random Each Loop:NiagaraBool=false
     Randomness Mode:ENiagaraRandomnessMode=::NewEnumerator0
     Random Seed:NiagaraInt32=?
     Fixed Random Seed:NiagaraBool=false
     Vector4 Selection Array:NiagaraDataInterfaceArrayFloat4=DataInterface() {propertyValues {GpuSyncMode=SyncCpuToGpu, MaxElements=0, FloatData [<6 10 1 0>, <6 10 0 0>]}}
    )
   ) {isSetParameters=true, module=SetVariables_#2}
   &FN_Light_Attributes(
    Color=<rgba 2 0 0.081 1>
    Create Color Attribute=true
    Create Exponent Attribute=true
    Create Radius Attribute=true
    Exponent=2
    Radius=50
    Scale Color=10
   )
   /SetVariables(Particles.DecalCamOffset=20, Particles.DecalColor:LinearColor=<rgba 1.1493 0.0149 0 0.5>, Particles.DecalDyn0:Vector4f=<1 0 1 0>) {isSetParameters=true, module=SetVariables_#3}
   /SetVariables(
    Particles.GlowColor=<rgba 0.6892 0.0336 0 1.5>
    Particles.GlowSpriteSize:Vector2f=&Vector2DFromFloat(Value=120)
   ) {isSetParameters=true, module=SetVariables_#4}
   /SetVariables(Particles.DetailColor:LinearColor=<rgba 1 0.1513 0.1033 0.5>) {isSetParameters=true, module=SetVariables_#5}
   &CameraOffset(Camera Offset Amount=&Lerp_Float(A=-6, B=-15, Alpha=$User.ParaGlideLeanAlpha))
  ]
  rendererClasses [
   &NiagaraMeshRendererProperties
   &NiagaraLightRendererProperties
   &NiagaraSpriteRendererProperties
  ]
  renderers [
   {rendererClass=&NiagaraMeshRendererProperties, rendererIndex=0}
   {rendererClass=&NiagaraMeshRendererProperties, rendererIndex=1}
   {rendererClass=&NiagaraLightRendererProperties, rendererIndex=2}
   {rendererClass=&NiagaraSpriteRendererProperties, rendererIndex=3}
   {rendererClass=&NiagaraSpriteRendererProperties, rendererIndex=4}
   {rendererClass=&NiagaraMeshRendererProperties, rendererIndex=5}
  ]
 }
 LightStreak_Hand {
  emitterUpdateScript [
   &EmitterState(Life Cycle Mode=::NewEnumerator1, Loop Behavior=::NewEnumerator1, Loop Duration Mode=::NewEnumerator1, UseLoopDelay=false)
   &SpawnBurst_Instantaneous(Spawn Count=2, Use Spawn Probability=false, UseLoopCountLimit=false)
  ]
  particleSpawnScript [
   &PartitionParticles()
   &InitializeParticle(
    Color=<rgba 1 0.5473 0.4674 0.25>
    Mesh Scale Mode=::NewEnumerator4
    Sprite Size Mode=::NewEnumerator3
    Uniform Sprite Size=50
    UsePositionOffset=false
    Write Material Random=true
   )
   &InitialMeshOrientation(
    Mesh Orientation Mode=::NewEnumerator7
    Rotation=&SelectVectorFromArray(
     Direct Array Index=$Particles.PartitionParticles.Partition
     Vector Selection Array=DataInterface() {propertyValues {FloatData [<0.025 0 0>, <0.025 0 0>]}}
    )
    Use Rotation Vector=true
   ) {enabled=false}
   &DynamicMaterialParameters(
    Index 0 Param 1=80
    Index 0 Param 2=5
    Index 0 Param 3=&RandomRangeFloat(Minimum=-1, Maximum=1)
    Index 0 Param 4=1
    Module.Index0Param0Write=true
    Module.Index0Param1Write=true
    Module.Index0Param2Write=true
    Module.Index0Param3Write=true
   )
  ]
  particleUpdateScript [
   &ParticleState(Kill Particles When Lifetime Has Elapsed=false, Let Infinitely Lived Particles Die When Emitter Deactivates=true)
   &ScaleMeshSize(Scale Factor=&Lerp_Vector(
     A=<1 1 0>
     B=<1 1 1>
     Alpha=&FloatFromCurve(
      FloatCurve=DataInterface() {propertyValues {Curve {keys [{interpMode=RCIM_Cubic, leaveTangent=0.4055, tangentMode=RCTM_Break, time=0, value=0}, <at 0.5 0.95>]}}}
      CurveIndex=$Emitter.Age
     )
    ))
   &SkeletalMeshLocation(
    Bone / Socket Index=$Particles.PartitionParticles.Partition
    Bone Sampling Mode=::NewEnumerator4
    MaskSpawnGroup=false
    Orientation Sampling=::NewEnumerator1
    Skeletal Mesh=$System.Skeletal Mesh
   )
   &UpdateMeshOrientation(
    Facing Interpolation=false
    Orientation Method=::NewEnumerator1
    Override Quat=false
    TransformFacing=false
    TransformUp=false
    Up Direction:Vector3f=&FN_FrontEndVector3Modifier(Gameplay=<0 0 1>, Front End=&Lerp_Vector(A=<0 -0.3 1>, B=<0 -0.6 1>, Alpha=$User.ParaGlideLeanAlpha))
   )
   /SetVariables(
    Particles.BorderColor:LinearColor=<rgba 2 0.0367 0 0.504>
    Particles.BorderDyn0:Vector4f=<1 0.8002 0 0>
    Particles.BorderDyn1:Vector4f=&SelectVector4_FromArray(
     Array Sampling Mode:ENiagaraArraySamplingMode=::NewEnumerator1
     Direct Array Index:NiagaraInt32=$Particles.PartitionParticles.Partition
     Interpolation:NiagaraFloat=?
     Direct Array Mode:ENiagara_WrapClamp=::NewEnumerator0
     Evaluation Type:ENiagaraRandomnessEvaluation=::NewEnumerator0
     Normalized Interpolation:NiagaraBool=true
     Recalculate Random Each Loop:NiagaraBool=false
     Randomness Mode:ENiagaraRandomnessMode=::NewEnumerator0
     Random Seed:NiagaraInt32=?
     Fixed Random Seed:NiagaraBool=false
     Vector4 Selection Array:NiagaraDataInterfaceArrayFloat4=DataInterface() {propertyValues {GpuSyncMode=SyncCpuToGpu, MaxElements=0, FloatData [<15 20 1 0>, <15 20 0 0>]}}
    )
   ) {isSetParameters=true, module=SetVariables_#2}
   &FN_Light_Attributes(
    Color=<rgba 2 0 0.081 1>
    Create Color Attribute=true
    Create Exponent Attribute=true
    Create Radius Attribute=true
    Exponent=2
    Radius=35
    Scale Color=10
   )
   /SetVariables(Particles.DecalCamOffset=15, Particles.DecalColor:LinearColor=<rgba 1.1493 0.0149 0 1>, Particles.DecalDyn0:Vector4f=<1 0 1 0>) {isSetParameters=true, module=SetVariables_#3}
   /SetVariables(
    Particles.GlowCamOffset=5
    Particles.GlowColor=<rgba 0.6892 0.0336 0 2>
    Particles.GlowSpriteSize:Vector2f=&Vector2DFromFloat(Value=120)
   ) {isSetParameters=true, module=SetVariables_#4}
   /SetVariables(
    Particles.DetailColor:LinearColor=<rgba 1 0.1513 0.1033 0.7>
    Particles.DetailDynPar:Vector4f=&MakeVector4(X=1, Y=&Lerp_Float(A=0, B=$Engine.Time, Alpha=$User.ParaGlideLeanAlpha), Z=-1, W=0)
   ) {isSetParameters=true, module=SetVariables_#5}
  ]
  rendererClasses [
   &NiagaraMeshRendererProperties
   &NiagaraLightRendererProperties
   &NiagaraSpriteRendererProperties
  ]
  renderers [
   {rendererClass=&NiagaraMeshRendererProperties, rendererIndex=0}
   {rendererClass=&NiagaraMeshRendererProperties, rendererIndex=1}
   {rendererClass=&NiagaraLightRendererProperties, rendererIndex=2}
   {rendererClass=&NiagaraSpriteRendererProperties, rendererIndex=3}
   {rendererClass=&NiagaraSpriteRendererProperties, rendererIndex=4}
   {rendererClass=&NiagaraMeshRendererProperties, rendererIndex=5}
  ]
 }
 RedLines {
  emitterUpdateScript [
   &EmitterState()
   &SpawnRate(SpawnRate=&Lerp_Float(A=25, B=40, Alpha=$User.ParaGlideLeanAlpha), Use Spawn Probability=false)
  ]
  particleSpawnScript [
   &PartitionParticles(Number Of Partitions=4)
   &InitializeParticle(
    Color=<rgba 0.6701 0 0.0179 1>
    Lifetime Max=0.6
    Lifetime Min=0.4
    Lifetime Mode=::NewEnumerator1
    Sprite Size Mode=::NewEnumerator1
    Uniform Sprite Size Min=2
    UsePositionOffset=false
    Write Material Random=true
   )
   &SkeletalMeshLocation(
    Bone / Socket Index=$Particles.PartitionParticles.Partition
    Bone Sampling Mode=::NewEnumerator4
    MaskSpawnGroup=false
    Skeletal Mesh=$System.Skeletal Mesh
   )
   &ShapeLocation(Disc Coverage=0.3968, MaskSpawnGroup=false, Override Seed=false, Ring Radius=14, Shape Primitive=::NewEnumerator4, UseOwnerScale=true)
   &AddVelocity(
    Rotation Coordinate Space=::Simulation
    Rotation Mode=::NewEnumerator2
    Velocity=&FN_FrontEndVector3Modifier(
     Gameplay=&Lerp_Vector(
      A=&RandomRangeVector(Minimum=<0 0 150>, Maximum=<0 0 300>)
      B=&RandomRangeVector(Minimum=<0 -50 500>, Maximum=<0 -100 600>)
      Alpha=$User.ParaGlideLeanAlpha
     )
     Front End=&Lerp_Vector(A=<0 -150 450>, B=<1 -500 1000>, Alpha=$User.ParaGlideLeanAlpha)
    )
   )
  ]
  particleUpdateScript [
   &ParticleState()
   &SolveForcesAndVelocity(Clamp Velocity=false, Limit Acceleration=false)
   &ScaleSpriteSizeBySpeed(Max Scale Factor=<0.5 10>, Min Scale Factor=<0.2 2>)
   /SetVariables(
    Particles.CoreColor:LinearColor=<rgba 2 0.2804 0.1874 1>
    Particles.CoreSize:Vector2f=&Multiply_Vector2D(A=$Particles.SpriteSize, B=<0.3 0.6>)
   ) {isSetParameters=true}
  ]
  rendererClasses [&NiagaraSpriteRendererProperties]
  renderers [
   {rendererClass=&NiagaraSpriteRendererProperties}
   {rendererClass=&NiagaraSpriteRendererProperties, rendererIndex=1}
  ]
 }
 SpeedLinesFixed {
  emitterUpdateScript [
   &EmitterState(Life Cycle Mode=::NewEnumerator1, Loop Behavior=::NewEnumerator1, Loop Duration Mode=::NewEnumerator1, UseLoopDelay=false)
   &SpawnBurst_Instantaneous(Spawn Count=2, Use Spawn Probability=false, UseLoopCountLimit=false)
  ]
  particleSpawnScript [
   &PartitionParticles()
   &InitializeParticle(
    Color=<rgba 1 0.5473 0.4674 0.25>
    Mesh Scale=&SelectVectorFromArray(
     Direct Array Index=$Particles.PartitionParticles.Partition
     Vector Selection Array=DataInterface() {propertyValues {FloatData [<1 1 1.1>, <1 1 1.1>, <1 1 0.5>, <1 1 0.5>]}}
    )
    Mesh Scale Mode=::NewEnumerator4
    Sprite Size Mode=::NewEnumerator3
    Uniform Sprite Size=32
    UsePositionOffset=false
    Write Material Random=true
   )
   &InitialMeshOrientation(
    Mesh Orientation Mode=::NewEnumerator7
    Rotation=&SelectVectorFromArray(
     Direct Array Index=$Particles.PartitionParticles.Partition
     Vector Selection Array=DataInterface() {propertyValues {FloatData [<0.025 0.015 0>, <0.025 -0.015 0>]}}
    )
    Use Rotation Vector=true
   ) {enabled=false}
   &DynamicMaterialParameters(
    Index 0 Param 1=80
    Index 0 Param 2=5
    Index 0 Param 3=&RandomRangeFloat(Minimum=-1, Maximum=1)
    Index 0 Param 4=1
    Module.Index0Param0Write=true
    Module.Index0Param1Write=true
    Module.Index0Param2Write=true
    Module.Index0Param3Write=true
   )
  ]
  particleUpdateScript [
   &ParticleState(Kill Particles When Lifetime Has Elapsed=false, Let Infinitely Lived Particles Die When Emitter Deactivates=true)
   &ScaleMeshSize(Scale Factor=&Lerp_Vector(
     A=<1 1 0>
     B=<1 1 1>
     Alpha=&FloatFromCurve(
      FloatCurve=DataInterface() {propertyValues {Curve {keys [{interpMode=RCIM_Cubic, leaveTangent=0.4055, tangentMode=RCTM_Break, time=0, value=0}, <at 0.5 0.95>]}}}
      CurveIndex=$Emitter.Age
     )
    ))
   &SkeletalMeshLocation(
    Bone / Socket Index=$Particles.PartitionParticles.Partition
    Bone Sampling Mode=::NewEnumerator4
    MaskSpawnGroup=false
    Orientation Sampling=::NewEnumerator1
    Skeletal Mesh=$System.Skeletal Mesh
   )
   &UpdateMeshOrientation(
    Facing Interpolation=false
    Orientation Method=::NewEnumerator1
    Override Quat=false
    TransformFacing=false
    TransformUp=false
    Up Direction:Vector3f=&FN_FrontEndVector3Modifier(Gameplay=<0 0 1>, Front End=&Lerp_Vector(A=<0 -0.3 1>, B=<0 -0.6 1>, Alpha=$User.ParaGlideLeanAlpha))
   )
   /SetVariables(
    Particles.BorderColor:LinearColor=<rgba 2 0.0367 0 0.3>
    Particles.BorderDyn0:Vector4f=<1 0 0 0>
    Particles.BorderDyn1:Vector4f=&SelectVector4_FromArray(
     Array Sampling Mode:ENiagaraArraySamplingMode=::NewEnumerator1
     Direct Array Index:NiagaraInt32=$Particles.PartitionParticles.Partition
     Interpolation:NiagaraFloat=?
     Direct Array Mode:ENiagara_WrapClamp=::NewEnumerator0
     Evaluation Type:ENiagaraRandomnessEvaluation=::NewEnumerator0
     Normalized Interpolation:NiagaraBool=true
     Recalculate Random Each Loop:NiagaraBool=false
     Randomness Mode:ENiagaraRandomnessMode=::NewEnumerator0
     Random Seed:NiagaraInt32=?
     Fixed Random Seed:NiagaraBool=false
     Vector4 Selection Array:NiagaraDataInterfaceArrayFloat4=DataInterface() {propertyValues {GpuSyncMode=SyncCpuToGpu, MaxElements=0, FloatData [<8 35 0 0>, <8 35 0 0>, <4 16 0 0>, <4 16 0 0>]}}
    )
   ) {isSetParameters=true, module=SetVariables_#2}
   &FN_Light_Attributes(
    Color=<rgba 2 0 0.081 1>
    Create Color Attribute=true
    Create Exponent Attribute=true
    Create Radius Attribute=true
    Exponent=2
    Radius=35
    Scale Color=10
   )
   /SetVariables(Particles.DecalCamOffset=15, Particles.DecalColor:LinearColor=<rgba 1.1493 0.0149 0 1>, Particles.DecalDyn0:Vector4f=<1 0 1 0>) {enabled=false, isSetParameters=true, module=SetVariables_#3}
   /SetVariables(
    Particles.GlowCamOffset=5
    Particles.GlowColor=<rgba 0.6892 0.0336 0 2>
    Particles.GlowSpriteSize:Vector2f=&Vector2DFromFloat(Value=80)
   ) {enabled=false, isSetParameters=true, module=SetVariables_#4}
   /SetVariables(Particles.DetailColor:LinearColor=<rgba 1 0.1513 0.1033 0.7>) {isSetParameters=true, module=SetVariables_#5}
  ]
  rendererClasses [&NiagaraMeshRendererProperties]
  renderers [
   {rendererClass=&NiagaraMeshRendererProperties, rendererIndex=0}
   {rendererClass=&NiagaraMeshRendererProperties, rendererIndex=1}
  ]
 }
 Speed_Lines {
  emitterUpdateScript [
   &EmitterState(Delay First Loop Only=true, Life Cycle Mode=::NewEnumerator1, Loop Delay=1.5, Loop Duration=1, UseLoopDelay=true)
   &SpawnRate(
    SpawnRate=&FN_FrontEndFloatModifier(Gameplay=&Multiply_Float(A=50, B=1), Front End=50)
    Use Spawn Probability=false
   ) {module=SpawnRate001}
  ]
  particleSpawnScript [
   &InitializeParticle(
    Color=<rgba 0.733 0.9995 1 0.12>
    Lifetime Max=0.25
    Lifetime Min=0.15
    Lifetime Mode=::NewEnumerator1
    Sprite Size Max=<3 7>
    Sprite Size Min=<2 4>
    Sprite Size Mode=::NewEnumerator2
    UsePositionOffset=true
    Write Material Random=true
   )
   &ShapeLocation(
    Cylinder Radius=&FN_FrontEndFloatModifier(Gameplay=300, Front End=100)
    MaskSpawnGroup=false
    Offset=&FN_FrontEndVector3Modifier(Gameplay=<0 0 32>, Front End=<0 50 -30>)
    Offset Mode=::NewEnumerator0
    Override Seed=false
    Shape Primitive=::NewEnumerator1
    Surface Only=false
    UseOwnerScale=false
   )
   &AddVelocity(
    Rotation Coordinate Space=::World
    Velocity=&FN_FrontEndVector3Modifier(
     Gameplay=&Add_Vector(
      A=&RandomRangeVector(Minimum=<0 0 1000>, Maximum=<0 0 1200>)
      B=&Multiply_Vector(A=$Engine.Owner.Velocity, B=&RandomRangeVector(Minimum=<-0.2 -0.2 0>, Maximum=<-0.5 -0.5 0>))
     )
     Front End=&TransformVector(Vector=&RandomRangeVector(Minimum=<0 -500 1000>, Maximum=<0 -600 1000>), Destination Space=::World)
    )
   )
  ]
  particleUpdateScript [
   &ParticleState()
   &ScaleSpriteSize(Scale Factor=&Vector2DFromCurve(
     Vector2Curve=DataInterface() {propertyValues {XCurve {keys [<at 0 0.5 RCIM_Cubic>,<at 0.5 1 RCIM_Cubic>]}, YCurve {keys [<at 0 0 RCIM_Cubic>,<at 0.5 1 RCIM_Cubic>]}}}
     ScaleCurve=&Vector2DFromFloat(Value=1)
    ))
   &SolveForcesAndVelocity(Clamp Velocity=false, Limit Acceleration=false)
   &ScaleSpriteSizeBySpeed(Max Scale Factor=<1 30>, Min Scale Factor=<0.01 0.05>, Velocity Threshold=1)
   &ScaleColor(
    Scale Alpha=&FloatFromCurve(
     FloatCurve=DataInterface() {propertyValues {Curve {keys [
         {arriveTangent=6.8397, interpMode=RCIM_Cubic, leaveTangent=6.8397, tangentMode=RCTM_User, time=0, value=0}
         <at 0.3 1 RCIM_Cubic>
         {arriveTangent=-2.221, interpMode=RCIM_Cubic, leaveTangent=-2.221, tangentMode=RCTM_User, time=1, value=0}
        ]}}}
     Scale Curve=$User.ParaGlideLeanAlpha
    )
    Scale RGB=&VectorFromFloat(Value=&FN_FrontEndFloatModifier(Gameplay=1, Front End=1))
    ScaleA=true
    ScaleRGB=true
   )
  ]
  rendererClasses [&NiagaraSpriteRendererProperties]
  renderers [{rendererClass=&NiagaraSpriteRendererProperties}]
 }
 Spline_Hands {
  emitterUpdateScript [
   &EmitterState(Life Cycle Mode=::NewEnumerator1, Loop Behavior=::NewEnumerator1, Loop Duration Mode=::NewEnumerator1, UseLoopDelay=false)
   &SpawnBurst_Instantaneous(Spawn Count=2, Use Spawn Probability=false, UseLoopCountLimit=false)
  ]
  particleSpawnScript [
   &PartitionParticles()
   &InitializeParticle(Color Mode=::NewEnumerator0, Mesh Scale Mode=::NewEnumerator3, Mesh Uniform Scale=0.14, UsePositionOffset=false, Write Material Random=true)
  ]
  particleUpdateScript [
   &ParticleState(
    Kill Particles When Lifetime Has Elapsed=false
    Let Infinitely Lived Particles Die When Emitter Deactivates=true
    Loop Particles Lifetime=true
   )
   &UpdateMeshOrientation(
    Facing Interpolation=false
    Orientation Method=::NewEnumerator1
    Override Quat=false
    Side Direction:Vector3f=<1 0 1>
    Stabilization Mode:ENiagara_OrientVectorCount=::NewEnumerator3
    TransformFacing=false
    TransformSide=false
    TransformUp=false
   ) {module=UpdateMeshOrientation001}
   &SkeletalMeshLocation(
    Bone / Socket Index=$Particles.PartitionParticles.Partition
    Bone Sampling Mode=::NewEnumerator4
    MaskSpawnGroup=false
    Orientation Sampling=::NewEnumerator1
    Sampled Position Coordinate Space=::Local
    Skeletal Mesh=$System.Skeletal Mesh
   )
   &DynamicMaterialParameters(
    Index 0 Param 1=&FloatFromCurve(
     FloatCurve=DataInterface() {propertyValues {Curve {
        defaultValue=1
        keys [
         {interpMode=RCIM_Cubic, tangentMode=RCTM_User, time=0, value=0}
         {arriveTangent=6.6972, interpMode=RCIM_Cubic, tangentMode=RCTM_Break, time=0.3, value=0.7}
         {arriveTangent=1.7647, interpMode=RCIM_Cubic, leaveTangent=1.7647, time=0.35, value=1.5}
         <at 0.47 1 RCIM_Cubic>
        ]
       }}}
     CurveIndex=$Emitter.Age
     Scale Curve=7
    )
    Index 0 Param 3=0.2
    Module.Index0Float0Write=true
    Module.Index0Float1Write=true
    Module.Index0Float2Write=true
    Module.Index0Float3Write=true
   )
   &ScaleColor(
    Scale Alpha=&Multiply_Float(
     A=&FloatFromCurve(
      FloatCurve=DataInterface() {propertyValues {Curve {
         defaultValue=0
         keys [
          {interpMode=RCIM_Cubic, tangentMode=RCTM_User, time=0, value=0}
          {arriveTangent=4.2878, interpMode=RCIM_Cubic, tangentMode=RCTM_Break, time=0.55, value=1}
         ]
        }}}
      CurveIndex=$Emitter.Age
      Scale Curve=2
     )
     B=&FloatFromCurve(
      FloatCurve=DataInterface() {propertyValues {Curve {
         defaultValue=1
         keys [
          {interpMode=RCIM_Cubic, tangentMode=RCTM_User, time=0, value=0.8}
          {interpMode=RCIM_Cubic, tangentMode=RCTM_User, time=0.5, value=1}
          {tangentMode=RCTM_User, time=1, value=0.8}
         ]
        }}}
      CurveIndex=Linked() {linkedVariable {name=NPC.FortniteNPC.TimeOfDay_0-1, type=NiagaraFloat}}
     )
    )
    ScaleA=true
    ScaleRGB=false
   ) {module=ScaleColor001}
  ]
  rendererClasses [&NiagaraMeshRendererProperties]
  renderers [{rendererClass=&NiagaraMeshRendererProperties, rendererIndex=0}]
 }
}
summary {
 systemName=NS_Contrail_SynthContact
 emitters [
  {
   emitterName=LightStreak_Hand
   rendererClasses [
    &NiagaraMeshRendererProperties
    &NiagaraLightRendererProperties
    &NiagaraSpriteRendererProperties
   ]
  }
  {emitterName=SpeedLinesFixed, rendererClasses [&NiagaraMeshRendererProperties]}
  {
   emitterName=LightStreak_Feet
   rendererClasses [
    &NiagaraMeshRendererProperties
    &NiagaraLightRendererProperties
    &NiagaraSpriteRendererProperties
   ]
  }
  {emitterName=Spline_Hands, rendererClasses [&NiagaraMeshRendererProperties]}
  {emitterName=Speed_Lines, rendererClasses [&NiagaraSpriteRendererProperties]}
  {emitterName=RedLines, rendererClasses [&NiagaraSpriteRendererProperties]}
  {emitterName=LightRay, rendererClasses [&NiagaraMeshRendererProperties]}
 ]
 userVariables [
  User.bIsFrontEnd {type=NiagaraBool}
  User.bIsFrontEndPreview {type=NiagaraBool}
  User.ParaGlideLeanAlpha {type=NiagaraFloat, defaultValue=0, description=""}
  User.PlayerSpeed {type=NiagaraFloat, defaultValue=0, description=""}
 ]
}
